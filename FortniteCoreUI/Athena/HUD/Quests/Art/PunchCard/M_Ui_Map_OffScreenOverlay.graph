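# Translucent UI material for a map off-screen overlay in the quest punch card art, producing a vertical gradient mask fixed to screen space.
# The Y of ScreenAlignedUVs is offset by a falloff position and scaled by a falloff strength, then clamped into a ramp that rises down the screen.
# The Color parameter's alpha scales that ramp before a final saturate.
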Material /FortniteCoreUI/Athena/HUD/Quests/Art/PunchCard/M_Ui_Map_OffScreenOverlay
graph {
 properties {
  bUsedWithStaticMesh=false
  blendMode=BLEND_Translucent
  materialDomain=MD_UI
  pixelDepthOffsetMode=PDOM_Legacy
  usedShadingModels=Default Lit
 }
 nodes [
  n1 = MaterialFunctionCall() {materialFunction=&ScreenAlignedUVs}
  n2 = Multiply(
   A=Subtract(A=ComponentMask(None=@n1."X 100%, Y 100%") {g=true}, B=ScalarParameter() {defaultValue=0.45, group=Item, parameterName=falloff_position})
   B=ScalarParameter() {defaultValue=10, group=Item, parameterName=falloff_Strenght}
  ) {constB=32}
  Saturate(None=Multiply(A=VectorParameter() {parameterName=Color, defaultValue=<rgba 1 1 1 1>}.A, B=Saturate(None=@n2)))
 ]
}
